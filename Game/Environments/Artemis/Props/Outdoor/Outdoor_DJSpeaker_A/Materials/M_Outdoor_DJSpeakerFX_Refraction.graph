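# Additive unlit refraction material for the DJ speaker effect, pulsing with audio-reactive beat and drop timestamps from a parameter collection.
# Sine ripples, offset per object by pivot position, combine with radial masks and beat pulses to drive refraction.
# Panned caustics in radial coordinates form a two-color glow that brightens on drops.
# Camera-offset vertex logic collapses the mesh to its pivot when inactive, on low quality or in forward shading.
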Material /Game/Environments/Artemis/Props/Outdoor/Outdoor_DJSpeaker_A/Materials/M_Outdoor_DJSpeakerFX_Refraction
graph {
 properties {
  blendMode=BLEND_Additive
  pixelDepthOffsetMode=PDOM_Legacy
  refractionMethod=RM_IndexOfRefraction
  shadingModel=MSM_Unlit
  usedShadingModels=Unlit
 }
 nodes [
  n1 = MaterialFunctionCall() {materialFunction=&ObjectPivotPoint}
  n2 = WorldPosition()
  n3 = NamedRerouteUsage()
  n4 = Constant()
  n5 = ShadingPathSwitch(Default=Ceil(None=Saturate(None=@n3)), Forward=@n4)
  Add(
   A=MaterialFunctionCall(Offset Amount=ScalarParameter() {defaultValue=20, parameterName=CameraOffset}) {materialFunction=/Engine/Functions/Engine_MaterialFunctions02/WorldPositionOffset/CameraOffset}.Camera Offset
   B=LinearInterpolate(A=Subtract(A=@n1.Object Pivot Location, B=@n2.XYZ), Alpha=QualitySwitch(Default=@n5, Low=@n4)) {constB=0}
  )
  n6 = SphereMask(A=TextureCoordinate(), B=Constant() {r=0.5}) {attenuationRadius=0.75, hardnessPercent=0}
  n7 = Multiply(A=Saturate(None=LinearInterpolate(Alpha=@n6) {constA=-1, constB=2}), B=Add(A=NamedRerouteUsage(), B=NamedRerouteUsage()))
  DepthFade(Opacity=@n7) {fadeDistanceDefault=20}
  n8 = Time()
  Fmod(A=@n8, B=Constant() {r=1})
  n9 = ScalarParameter() {defaultValue=-0.05, parameterName=PulseRefraction}
  n10 = Subtract(A=@n8, B=CollectionParameter() {collection=/Game/Athena/Events/AudioAnalysis/MPC_Audio_Reactive_Parameters, parameterName=BeatGameTime})
  n11 = ScalarParameter() {defaultValue=0.8, parameterName=BeatDuration}
  n12 = OneMinus(None=Saturate(None=Divide(A=@n10, B=@n11)))
  n13 = NamedRerouteDeclaration(None=Add(A=Power(Base=@n12), B=NamedRerouteUsage())) {name=PulseAnimSlow, nodeColor=<rgba 1 0 0.1641 1>}
  n14 = Time()
  n15 = Divide(A=MaterialFunctionCall(f3=MaterialFunctionCall() {materialFunction=&ObjectPivotPoint}.Object Pivot Location) {materialFunction=&AddComponents}.f3) {constB=1000}
  n16 = Multiply(A=Subtract(A=Add(A=@n14, B=@n15), B=NamedRerouteUsage()), B=ScalarParameter() {defaultValue=1, parameterName=PanSpeed})
  n17 = Add(A=Multiply(A=@n16) {constB=0.5}, B=@n6)
  n18 = LinearInterpolate(A=ConstantBiasScale(None=Sine(None=@n17) {period=0.15}), B=Multiply(A=@n6, B=@n6)) {constAlpha=0.75}
  LinearInterpolate(
   B=Add(B=Multiply(A=@n9, B=@n13)) {constA=1}
   Alpha=Multiply(A=NamedRerouteDeclaration(None=@n18) {name=Ripple, nodeColor=<rgba 0.1953 0 1 1>}, B=@n7)
  ) {constA=1, constB=1.2}
  n19 = NamedRerouteUsage()
  n20 = Multiply(A=@n19, B=@n19)
  n21 = MaterialFunctionCall(Float2=MaterialFunctionCall() {materialFunction=&VectorToRadialValue}.Radial Coordinates) {materialFunction=&BreakOutFloat2Components}
  n22 = NamedRerouteUsage()
  n23 = Add(
   A=Multiply(A=Power(Base=@n21.G) {constExponent=1.6}) {constB=0.5}
   B=Multiply(A=NamedRerouteUsage(), B=Multiply(A=@n22) {constB=0.5}) {constB=0.2}
  )
  n24 = TextureCoordinate()
  n25 = TextureSample!(
   UVs=Panner(Coordinate=AppendVector(A=@n21.R, B=@n23), Time=@n16) {speedX=0.1, speedY=-0.25}
   Tex
   "DDX(UVs)"=DDX(Value=@n24)
   "DDY(UVs)"=DDY(Value=@n24)
   Apply View MipBias
  ) {mipValueMode=TMVM_Derivative, texture=/Game/Effects/Fort_Effects/Textures/Tile/T_ShieldCaustics_Tile_Chromatic}
  n26 = SphereMask(A=TextureCoordinate(), B=Constant() {r=0.5}) {attenuationRadius=0.5, hardnessPercent=10}
  n27 = VectorParameter() {parameterName=Color1, defaultValue=<rgba 0.1199 0.1344 1.5 0>}
  n28 = VectorParameter() {parameterName=Color2, defaultValue=<rgba 2.0658 0.8203 2.5 0>}
  n29 = NamedRerouteUsage()
  n30 = Time()
  n31 = CollectionParameter() {collection=&MPC_Audio_Reactive_Parameters, parameterName=DropGameTime}
  n32 = Saturate(None=Divide(A=Subtract(A=@n30, B=@n31), B=ScalarParameter() {defaultValue=1.2, parameterName=DropDuration}))
  Multiply(
   A=Multiply(A=LinearInterpolate(A=@n20, B=@n25.RGB) {constAlpha=0.8}, B=OneMinus(None=@n26))
   B=Multiply(A=LinearInterpolate(A=@n27.RGB, B=@n28.RGB, Alpha=@n29), B=LinearInterpolate(Alpha=@n32) {constA=10})
  )
  Multiply(A=VertexNormalWS()) {constB=-1}
  NamedRerouteDeclaration(None=Power(Base=OneMinus(None=@n32)) {constExponent=3}) {name=DropAnim, nodeColor=<rgba 0.3828 0 1 1>}
  NamedRerouteDeclaration(None=Power(Base=@n12) {constExponent=5}) {name=PulseAnim, nodeColor=<rgba 1 0.3047 0 1>}
  TextureSample!(UVs, Tex, Apply View MipBias) {samplerType=SAMPLERTYPE_Normal, texture=&T_Sphere_N}
 ]
}
