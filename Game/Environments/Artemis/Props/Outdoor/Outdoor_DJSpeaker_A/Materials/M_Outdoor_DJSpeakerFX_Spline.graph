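# Additive unlit material for DJ speaker spline effects that pulse and change color to the beats and drops of Artemis event music.
# Beat and drop times from the audio-reactive parameter collection drive decaying pulses, offset by vertex alpha so they travel along the mesh.
# SplineThicken sets line width, while world position offset pushes spikes outward, with vertex RGB and pivot-seeded sines varying which spikes appear.
# Drops blend width, color and displacement.
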Material /Game/Environments/Artemis/Props/Outdoor/Outdoor_DJSpeaker_A/Materials/M_Outdoor_DJSpeakerFX_Spline
graph {
 properties {blendMode=BLEND_Additive, pixelDepthOffsetMode=PDOM_Legacy, shadingModel=MSM_Unlit, usedShadingModels=Unlit}
 nodes [
  n1 = NamedRerouteUsage()
  n2 = Multiply(
   A=LinearInterpolate(
    A=ScalarParameter() {defaultValue=30, parameterName=Width}
    B=ScalarParameter() {defaultValue=60, parameterName=WidthDrop}
    Alpha=NamedRerouteUsage()
   )
   B=Ceil(None=Saturate(None=@n1))
  )
  n3 = LinearInterpolate(Alpha=NamedRerouteDeclaration(None=VertexColor().A) {name=HeightMask, nodeColor=<rgba 1 0.4453 0 1>}) {constA=-0.12, constB=0}
  n4 = Add(A=@n3, B=Time())
  n5 = CollectionParameter() {collection=/Game/Athena/Events/AudioAnalysis/MPC_Audio_Reactive_Parameters, parameterName=BeatGameTime}
  n6 = Saturate(None=Divide(A=Subtract(A=@n4, B=@n5), B=ScalarParameter() {defaultValue=0.6, parameterName=BeatDuration}))
  n7 = NamedRerouteDeclaration(None=Power(Base=OneMinus(None=@n6)) {constExponent=5}) {name=PulseAnim, nodeColor=<rgba 1 0.3047 0 1>}
  n8 = NamedRerouteUsage()
  n9 = NamedRerouteUsage()
  n10 = Multiply(
   A=Transform(None=Constant3Vector() {constant=<rgba 1 0 0 0>}) {transformSourceType=TRANSFORMSOURCE_Local}
   B=ScalarParameter() {defaultValue=140, parameterName=WPOPulseDist}
  )
  n11 = Multiply(
   A=Subtract(
    A=WorldPosition().XYZ
    B=MaterialFunctionCall() {materialFunction=&ObjectPivotPoint}.Object Pivot Location
   )
   B=ScalarParameter() {defaultValue=0.7, parameterName=WPOScaleMult}
  ) {constB=0.25}
  n12 = NamedRerouteUsage()
  n13 = NamedRerouteUsage()
  n14 = VectorParameter() {parameterName=SpikeSizes, defaultValue=<rgba 0.8 0.5 0.3 1>}
  n15 = VertexColor()
  n16 = MaterialFunctionCall() {materialFunction=&ObjectPivotPoint}
  n17 = Add(
   A=Time()
   B=Divide(A=MaterialFunctionCall(f3=@n16.Object Pivot Location) {materialFunction=&AddComponents}.f3) {constB=1000}
  )
  n18 = Sine(None=@n17)
  n19 = Sine(None=@n17) {period=1.5}
  n20 = Sine(None=@n17) {period=0.5}
  n21 = Clamp(None=MaterialFunctionCall(R=Multiply(A=@n15.R, B=@n18), G=Multiply(A=@n15.G, B=@n19), B=Multiply(A=@n15.B, B=@n20)) {materialFunction=&AppendMany}.RGB) {clampMode=CMODE_ClampMin}
  n22 = MaterialFunctionCall(f3=Multiply(A=Multiply(A=@n14.RGB, B=@n15), B=Ceil(None=@n21))) {materialFunction=&AddComponents}
  n23 = Multiply(A=Multiply(A=Add(A=@n10, B=@n11), B=LinearInterpolate(Alpha=@n12) {constA=1, constB=0.5}), B=Add(B=Multiply(A=@n13, B=@n22.f3)) {constA=0.5})
  n24 = ScalarParameter() {defaultValue=1, parameterName=WPOMult}
  n25 = LinearInterpolate(A=@n24, B=Multiply(A=@n24) {constB=3.5}, Alpha=NamedRerouteUsage())
  Add(
   A=MaterialFunctionCall(WidthBase=@n2, WidthTip=@n2) {materialFunction=&SplineThicken}.WorldPositionOffset
   B=Multiply(A=LinearInterpolate(A=@n7, B=@n8, Alpha=@n9), B=Multiply(A=@n23, B=@n25))
  )
  Constant() {r=1}
  Constant()
  Constant()
  n26 = Constant() {r=1}
  Fmod(A=Add(A=@n3, B=Time()), B=@n26)
  n27 = NamedRerouteUsage()
  Multiply(
   A=Multiply(A=LinearInterpolate(Alpha=@n27) {constA=0.1}, B=NamedRerouteUsage())
   B=MaterialFunctionCall("FadeSensitivity (optional)"=@n26) {materialFunction=&FadeBasedOnViewAngle}.Result
  )
  Multiply(A=@n13, B=Add(A=@n22.f3, B=ScalarParameter() {defaultValue=0.5, parameterName=MinBrightness}))
  NamedRerouteDeclaration(None=@n13) {name=BrightnessAnim, nodeColor=<rgba 0.8281 1 0 1>}
  n28 = Subtract(A=@n4, B=CollectionParameter() {collection=&MPC_Audio_Reactive_Parameters, parameterName=DropGameTime})
  n29 = ScalarParameter() {defaultValue=1, parameterName=DropDuration}
  n30 = OneMinus(None=Saturate(None=Divide(A=@n28, B=@n29)))
  NamedRerouteDeclaration(None=Power(Base=@n30) {constExponent=0.6}) {name=Dropping, nodeColor=<rgba 1 0 0.1172 1>}
  NamedRerouteDeclaration(None=Power(Base=@n30) {constExponent=3}) {name=DropAnim, nodeColor=<rgba 0.9688 1 0 1>}
  n31 = VectorParameter() {parameterName=ColorBase, defaultValue=<rgba 1 0.0709 0.8099 0>}
  n32 = VectorParameter() {parameterName=DropColor, defaultValue=<rgba 0.1412 2.2443 3 0>}
  n33 = NamedRerouteUsage()
  n34 = Constant() {r=0.5}
  n35 = SphereMask(A=ComponentMask(None=TextureCoordinate()) {r=true}, B=@n34, Radius=@n34) {hardnessPercent=0}
  n36 = Saturate(None=LinearInterpolate(Alpha=@n35) {constA=-2})
  n37 = Multiply(
   A=LinearInterpolate(
    A=VectorParameter() {parameterName=ColorEnd', defaultValue=<rgba 0.08 0.0896 1 0>}.RGB
    B=LinearInterpolate(A=@n31.RGB, B=@n32.RGB, Alpha=@n33)
    Alpha=NamedRerouteUsage()
   )
   B=Power(Base=LinearInterpolate(A=@n35, B=@n36) {constAlpha=0.25}) {constExponent=3}
  )
  ShadingPathSwitch(
   Default=MaterialFunctionCall(
    LDR Input=Clamp(None=@n37)
    HDR Tint=@n37
    Derive HDR Power=ScalarParameter() {defaultValue=3, parameterName=HDRMult}
    Desaturate Luminance=Constant() {r=0.05}
   ) {materialFunction=&DeriveHDRfromLDR}.RemappedRange
   Forward=@n37
   Mobile=@n37
  )
 ]
}
